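# Sound cue named as a death stinger for henchmen in the HeistBaseGuards content, assigned to the NPC music sound class.
# The cue has no root and no sound nodes, so as written it produces no audio.
# It carries only playback settings: a play-when-silent virtualization mode and a very large max distance.
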SoundCue /HeistBaseGuards/Audio/Henchmen_Death_Stinger_Cue
root=null
nodes []
properties {
 maxDistance=2097152
 soundClassObject=&Fort_Music_NPC
 virtualizationMode=PlayWhenSilent
 volumeMultiplier=1
}
