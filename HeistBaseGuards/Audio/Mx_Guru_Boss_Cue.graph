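# Adaptive music SoundCue for the Guru boss in HeistBaseGuards, assigned to the NPC music sound class.
# A mixer blends three branches whose volumes follow the continuous parameters Unaware, Alerted and Threatened, each defaulting to silent.
# Unaware drives the looping idle track and Threatened the looping aggro track, while the Alerted branch has boosted gain but no wave connected.
# PlayWhenSilent virtualization keeps the cue playing while inaudible.
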SoundCue /HeistBaseGuards/Audio/Mx_Guru_Boss_Cue
root=Mixer_1
nodes [
 {id=Mixer_1, children [ModulatorContinuous_0, ModulatorContinuous_1, ModulatorContinuous_3], properties {inputVolume [1,2,1]}}
 {id=ModulatorContinuous_0, children [Modulator_2], properties {volumeModulationParams {default=0, parameterName=Unaware}}}
 {id=ModulatorContinuous_1, children [Modulator_1], properties {volumeModulationParams {default=0, parameterName=Alerted}}}
 {id=ModulatorContinuous_3, children [Modulator_0], properties {volumeModulationParams {default=0, parameterName=Threatened}}}
 {id=Modulator_0, children [WavePlayer_2], properties {pitchMax=1, pitchMin=1, volumeMax=1, volumeMin=1}}
 {id=Modulator_1, children [null], properties {pitchMax=1, pitchMin=1, volumeMax=1.3, volumeMin=1.3}}
 {id=Modulator_2, children [WavePlayer_0], properties {pitchMax=1, pitchMin=1, volumeMax=1, volumeMin=1}}
 {id=WavePlayer_0, properties {bLooping=true, soundWaveAssetPtr=/HeistBaseGuards/Audio/wav/GuruBoss_Idle_Loop}}
 {id=WavePlayer_2, properties {bLooping=true, soundWaveAssetPtr=/HeistBaseGuards/Audio/wav/GuruBoss_Aggro_Loop}}
]
properties {
 maxDistance=2097152
 soundClassObject=&Fort_Music_NPC
 virtualizationMode=PlayWhenSilent
 volumeMultiplier=1.3
}
