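# Looping adaptive gameplay music cue for the heist base guards, with layers driven by Unaware and Threatened parameters.
# Unaware modulates the henchmen track's volume, while Threatened modulates a submix of that same track plus the combat track.
# Both continuous modulators default to zero and feed the root mixer.
# The cue keeps playing when silent and overrides concurrency to two instances, rejecting new ones beyond that.
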SoundCue /HeistBaseGuards/Audio/Mx_Heist_Heisting_LP_Cue
root=Mixer_1
nodes [
 {id=Mixer_0, children [WavePlayer_1,WavePlayer_2], properties {inputVolume [1,1]}}
 {id=Mixer_1, children [ModulatorContinuous_0, ModulatorContinuous_3], properties {inputVolume [1,1]}}
 {id=ModulatorContinuous_0, children [WavePlayer_0], properties {volumeModulationParams {default=0, parameterName=Unaware}}}
 {id=ModulatorContinuous_3, children [Mixer_0], properties {volumeModulationParams {default=0, parameterName=Threatened}}}
 {id=WavePlayer_0, properties {bLooping=true, soundWaveAssetPtr=/HeistBaseGuards/Audio/wav/mx_Heist_Heisting_Henchmen_Mastered}}
 {id=WavePlayer_1, properties {bLooping=true, soundWaveAssetPtr=&mx_Heist_Heisting_Henchmen_Mastered}}
 {id=WavePlayer_2, properties {bLooping=true, soundWaveAssetPtr=/HeistBaseGuards/Audio/wav/mx_Heist_Heisting_Combat_Mastered}}
]
properties {
 bOverrideConcurrency=true
 maxDistance=2097152
 soundClassObject=/Game/Sounds/FortSoundClassesAndModes/Fort_Music_Gameplay
 virtualizationMode=PlayWhenSilent
 volumeMultiplier=1.7
 concurrencyOverrides {maxCount=2, resolutionRule=PreventNew, voiceStealReleaseTime=0.2, volumeScaleAttackTime=0.2}
}
