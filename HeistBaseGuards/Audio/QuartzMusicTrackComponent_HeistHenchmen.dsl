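; Quartz music track component for the henchmen in the heist base guards content, judging by its name and path.
; It overrides IsEnabled to always return true, so the track is never disabled through that check.
; BeginPlay only calls the parent, and the file itself holds no other logic.
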(event EventBeginPlay
  (|CallParentFunction))

(fn IsEnabled
  (return true))
